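# Niagara system for the Barium Female glider's cosmetic effects: trails, spinning wheels and speed lines.
# Ribbon and wheel emitters follow User.SkeletalMesh, tinted by BariumPaintColor, with wheel meshes rotated over time by quaternions.
# bIsFrontEndPreview switches between two ribbon pairs with different motion inheritance and retunes speed lines, which respect bIsFullyDeployed in preview.
# A system integrator value scales alpha, while a RightDot threshold drives persistent sideways-thrust audio.
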NiagaraSystem /BRCosmetics/Gadgets/Assets/VinderTech_GliderChute/Glider_Barium_Female/FX/NS_Glider_Barium
dependencies {
 usedDataInterfaces [
  &NiagaraDataInterfaceVector2DCurve
  &NiagaraDataInterfaceCurve
  &NiagaraDataInterfaceColorCurve
  &NiagaraDataInterfaceAudioPlayer
  &NiagaraDataInterfaceArrayFloat
 ]
 usedDynamicInputs []
 usedModules [
  &self:SystemSpawnScript.SystemScriptSource.SystemScriptGraph.NiagaraNodeAssignment_0.SetVariables_#1
  &SystemState
 ]
}
emitters {
 DEBUG {
  bEnabled=false
  emitterUpdateScript [
   &EmitterState()
   &SpawnBurst_Instantaneous(Spawn Count=4, Use Spawn Probability=false, UseLoopCountLimit=false)
  ]
  particleSpawnScript [
   &FN_SetSpawnGroup()
   &InitializeParticle(Color Mode=::NewEnumerator0, Lifetime=5, UsePositionOffset=false, Write Material Random=true)
  ]
  particleUpdateScript [
   &ParticleState()
   &SkeletalMeshLocation(Bone / Socket Index=$Particles.SpawnGroup, Bone Sampling Mode=::NewEnumerator4, MaskSpawnGroup=false, Skeletal Mesh=$User.SkeletalMesh)
  ]
  rendererClasses [&NiagaraSpriteRendererProperties]
  renderers [{rendererClass=&NiagaraSpriteRendererProperties}]
 }
 FN_AttachedRibbon {
  emitterUpdateScript [
   &EmitterState()
   &FN_SpawnRibbon_v2(
    Number of Ribbons:NiagaraInt32=&MakeCustomIntFromBool(Bool=$User.bIsFrontEndPreview, True INT=4, False INT=0)
    SpawnRate:NiagaraFloat=&MakeCustomFloatFromBool(Bool=$User.bIsFrontEndPreview, True Float=30, False Float=0)
   )
   &FN_Integrator() {enabled=false}
  ]
  particleSpawnScript [
   /Game/Effects/Niagara/Modules/Ribbons/FN_InitializeRibbon_v4(Enable Skeletal Position=true, Lifetime=0.4, Ribbon Width=13, SkeletalMesh:NiagaraDataInterfaceSkeletalMesh=$User.SkeletalMesh)
   &FN_FlipRibbonV()
   &AddVelocity(Velocity=<0 200 0>)
  ]
  particleUpdateScript [
   &InheritSourceMovement(
    Applied Force Scale=<0.75 0.75 0.9>
    Applied Position Scale=&VectorFromFloat(Value=&FloatFromCurve(FloatCurve=DataInterface() {propertyValues {Curve {keys [<at 0.5 0.97>,<at 1 0.9>]}}}))
    ApplyForce=false
    ApplyPosition=true
    ApplyPositionFalloff=false
   )
   &SolveForcesAndVelocity(Clamp Velocity=false, Limit Acceleration=false)
   /Game/Effects/Niagara/Modules/Ribbons/FN_UpdateRibbon_v4(Enable Skeletal Position=true, SkeletalMesh:NiagaraDataInterfaceSkeletalMesh=$User.SkeletalMesh)
   &FN_RibbonMaterial(
    Color=$User.BariumPaintColor
    Scale Alpha:NiagaraFloat=&FloatFromCurve(
     FloatCurve=DataInterface() {propertyValues {Curve {keys [<at 0 1>,<at 1 0>]}}}
     Scale Curve=&Lerp_Float(A=0.5, B=1, Alpha=$System.FN_Integrator001.Value)
    )
   )
  ]
  rendererClasses [&NiagaraRibbonRendererProperties]
  renderers [{rendererClass=&NiagaraRibbonRendererProperties}]
 }
 FN_AttachedRibbon001 {
  emitterUpdateScript [
   &EmitterState()
   &FN_SpawnRibbon_v2(
    Number of Ribbons:NiagaraInt32=&MakeCustomIntFromBool(Bool=$User.bIsFrontEndPreview, True INT=4, False INT=0)
    SpawnRate:NiagaraFloat=&MakeCustomFloatFromBool(Bool=$User.bIsFrontEndPreview, True Float=30, False Float=0)
   )
   &FN_Integrator() {enabled=false}
   &FN_Timer()
  ]
  particleSpawnScript [
   &FN_InitializeRibbon_v4(
    Color:LinearColor=$User.BariumPaintColor
    Enable Skeletal Position=true
    Ribbon Width=13
    SkeletalMesh:NiagaraDataInterfaceSkeletalMesh=$User.SkeletalMesh
   )
   &FN_FlipRibbonV()
   &AddVelocity(Velocity=<0 200 0>)
  ]
  particleUpdateScript [
   &InheritSourceMovement(
    Applied Force Scale=<0.5 0.5 0.9>
    Applied Position Scale=&VectorFromFloat(Value=&FloatFromCurve(FloatCurve=DataInterface() {propertyValues {Curve {keys [<at 0.5 0.97>,<at 1 0.9>]}}}))
    ApplyForce=false
    ApplyPosition=true
    ApplyPositionFalloff=false
   )
   &SolveForcesAndVelocity(Clamp Velocity=false, Limit Acceleration=false)
   &FN_UpdateRibbon_v4(Enable Skeletal Position=true, SkeletalMesh:NiagaraDataInterfaceSkeletalMesh=$User.SkeletalMesh)
   &DynamicMaterialParameters(
    Index 3 Param 1=&Add_Float(
     A=&SelectFloatFromArray(Direct Array Index=$Particles.SpawnGroup, Float Selection Array=DataInterface() {propertyValues {FloatData [1,2,3,4]}})
     B=$Emitter.FN_Timer.Value
    )
    Module.Param0WriteEnabled3=true
    Module.Param1WriteEnabled3=false
    Module.Param2WriteEnabled3=false
    Module.Param3WriteEnabled3=false
    Write Parameter Index 0=false
    Write Parameter Index 3=true
   )
   &ScaleColor(
    Scale Alpha=&FloatFromCurve(
     FloatCurve=DataInterface() {propertyValues {Curve {keys [<at 0.5 1>,<at 1 0>]}}}
     Scale Curve=&Lerp_Float(A=0.5, B=1, Alpha=$System.FN_Integrator001.Value)
    )
    ScaleA=true
    ScaleRGB=true
   )
  ]
  rendererClasses [&NiagaraRibbonRendererProperties]
  renderers [{rendererClass=&NiagaraRibbonRendererProperties}]
 }
 FN_AttachedRibbon002 {
  emitterUpdateScript [
   &EmitterState()
   &FN_SpawnRibbon_v2(
    Number of Ribbons:NiagaraInt32=&MakeCustomIntFromBool(Bool=$User.bIsFrontEndPreview, True INT=0, False INT=4)
    SpawnRate:NiagaraFloat=&MakeCustomFloatFromBool(Bool=$User.bIsFrontEndPreview, True Float=0, False Float=30)
   )
   &FN_Integrator() {enabled=false}
  ]
  particleSpawnScript [
   &FN_InitializeRibbon_v4(Enable Skeletal Position=true, Ribbon Width=13, SkeletalMesh:NiagaraDataInterfaceSkeletalMesh=$User.SkeletalMesh)
   /SetVariables(Particles.Position=&Add_Vector(
     A=Linked() {linkedVariable {name=Particles.Position, type=NiagaraPosition}}
     B=&MakeVector(
      X=0
      Y=0
      Z=&SelectFloatFromArray(Direct Array Index=$Particles.SpawnGroup, Float Selection Array=DataInterface() {propertyValues {FloatData [-10,0,0,0,0,0,0,0]}})
     )
    )) {enabled=false, isSetParameters=true}
   &FN_FlipRibbonV()
   &AddVelocity(Velocity=<0 100 0>)
  ]
  particleUpdateScript [
   &InheritSourceMovement(
    Applied Force Scale=<-0.25 -0.25 -0.1>
    Applied Position Scale=<0 1 0>
    ApplyForce=true
    ApplyForceFalloff=false
    ApplyPosition=false
    Module.LimitForceSpeed=false
    Source Velocity:Vector3f=&TransformVector(Vector=$Engine.Owner.Velocity, Source Space=::World, Destination Space=::Local)
   )
   &SolveForcesAndVelocity(Clamp Velocity=false, Limit Acceleration=false)
   &FN_UpdateRibbon_v4(Enable Skeletal Position=true, SkeletalMesh:NiagaraDataInterfaceSkeletalMesh=$User.SkeletalMesh)
   &FN_RibbonMaterial(
    Color=$User.BariumPaintColor
    Scale Alpha:NiagaraFloat=&FloatFromCurve(
     FloatCurve=DataInterface() {propertyValues {Curve {keys [<at 0 1>,<at 1 0>]}}}
     Scale Curve=&Lerp_Float(A=0.5, B=1, Alpha=$System.FN_Integrator001.Value)
    )
   )
  ]
  rendererClasses [&NiagaraRibbonRendererProperties]
  renderers [{rendererClass=&NiagaraRibbonRendererProperties}]
 }
 FN_AttachedRibbon003 {
  emitterUpdateScript [
   &EmitterState()
   &FN_SpawnRibbon_v2(
    Number of Ribbons:NiagaraInt32=&MakeCustomIntFromBool(Bool=$User.bIsFrontEndPreview, True INT=0, False INT=4)
    SpawnRate:NiagaraFloat=&MakeCustomFloatFromBool(Bool=$User.bIsFrontEndPreview, True Float=0, False Float=30)
   )
   &FN_Integrator() {enabled=false}
   &FN_Timer()
  ]
  particleSpawnScript [
   &FN_InitializeRibbon_v4(
    Color:LinearColor=$User.BariumPaintColor
    Enable Skeletal Position=true
    Lifetime=0.7
    Ribbon Width=13
    SkeletalMesh:NiagaraDataInterfaceSkeletalMesh=$User.SkeletalMesh
   )
   &FN_FlipRibbonV()
   &AddVelocity(Velocity=<0 100 0>)
  ]
  particleUpdateScript [
   &InheritSourceMovement(
    Applied Force Scale=<-0.25 -0.25 -0.1>
    Applied Position Scale=<0 1 0>
    ApplyForce=true
    ApplyForceFalloff=false
    ApplyPosition=false
    Module.LimitForceSpeed=false
    Source Velocity:Vector3f=&TransformVector(Vector=$Engine.Owner.Velocity, Source Space=::World, Destination Space=::Local)
   )
   &SolveForcesAndVelocity(Clamp Velocity=false, Limit Acceleration=false)
   &FN_UpdateRibbon_v4(Enable Skeletal Position=true, SkeletalMesh:NiagaraDataInterfaceSkeletalMesh=$User.SkeletalMesh)
   &DynamicMaterialParameters(
    Index 3 Param 1=&Add_Float(
     A=&SelectFloatFromArray(Direct Array Index=$Particles.SpawnGroup, Float Selection Array=DataInterface() {propertyValues {FloatData [1,2,3,4]}})
     B=$Emitter.FN_Timer.Value
    )
    Module.Param0WriteEnabled3=true
    Module.Param1WriteEnabled3=false
    Module.Param2WriteEnabled3=false
    Module.Param3WriteEnabled3=false
    Write Parameter Index 0=false
    Write Parameter Index 3=true
   )
   &ScaleColor(
    Scale Alpha=&FloatFromCurve(
     FloatCurve=DataInterface() {propertyValues {Curve {keys [<at 0.5 1>,<at 1 0>]}}}
     Scale Curve=&Lerp_Float(A=0.5, B=1, Alpha=$System.FN_Integrator001.Value)
    )
    ScaleA=true
    ScaleRGB=true
   )
  ]
  rendererClasses [&NiagaraRibbonRendererProperties]
  renderers [{rendererClass=&NiagaraRibbonRendererProperties}]
 }
 SpeedLines {
  emitterUpdateScript [
   &EmitterState()
   &SpawnRate(
    SpawnRate=&MakeCustomFloatFromBool(Bool=$User.bIsFrontEndPreview, True Float=50, False Float=90)
    Use Spawn Probability=true
   ) {module=SpawnRate001}
  ]
  particleSpawnScript [
   &InitializeParticle(
    Color=&MakeLinearColorFromVectorAndFloat(
     "Vector (RGB)"=<1 1 1>
     "Float (Alpha)"=&Multiply_Float(
      A=&UniformRangedFloat(Minimum=0.1, Maximum=0.18)
      B=&MakeCustomFloatFromBool(Bool=$User.bIsFrontEndPreview, True Float=0.5, False Float=1)
     )
    )
    Lifetime Max=0.16
    Lifetime Min=0.1
    Lifetime Mode=::NewEnumerator1
    Sprite Size Max=<3 10>
    Sprite Size Min=<2 2>
    Sprite Size Mode=::NewEnumerator2
    UsePositionOffset=false
    Write Material Random=true
   )
   &TorusLocation(
    Align Torus With Axis=true
    Handle Radius=0
    Large Radius:NiagaraFloat=&RandomRangeFloat(Minimum=80, Maximum=150)
    MaskSpawnGroup=false
    Module.Override Randomness=false
    Offset=<0 -180 0>
    Torus Axis=<0 1 0>
    Torus Origin:NiagaraPosition=&Add_Vector(
     A=Linked() {linkedVariable {name=Particles.Position, type=NiagaraPosition}}
     B=&MakeVector(X=0, Y=&MakeCustomFloatFromBool(Bool=$User.bIsFrontEndPreview, True Float=0, False Float=0), Z=0)
    )
   )
   &AddVelocityFromPoint(
    Module.Use Falloff Distance=false
    Origin Offset=<0 -500 0>
    Velocity Strength=&Multiply_Float(A=600, B=&MakeCustomFloatFromBool(Bool=$User.bIsFrontEndPreview, True Float=-1, False Float=1))
   ) {enabled=false}
   &AddVelocity(Velocity=&Multiply_Vector(
     A=&RandomRangeVector(Minimum=<0 300 0>, Maximum=<0 700 0>)
     B=&MakeVector(X=1, Y=&MakeCustomFloatFromBool(Bool=$User.bIsFrontEndPreview, True Float=-1, False Float=1), Z=1)
    )) {enabled=false}
   &AddVelocityFromPoint(Module.Use Falloff Distance=false, Origin Offset=<0 -500 0>, Velocity Strength=600) {module=AddVelocityFromPoint001}
   &AddVelocity(Velocity=&RandomRangeVector(Minimum=<0 300 0>, Maximum=<0 700 0>)) {module=AddVelocity001}
  ]
  particleUpdateScript [
   &SolveForcesAndVelocity(Clamp Velocity=false, Limit Acceleration=false)
   &ParticleState()
   &ScaleSpriteSize(Scale Factor=&Vector2DFromCurve(
     Vector2Curve=DataInterface() {propertyValues {XCurve {keys [<at 0 0.5 RCIM_Cubic>,<at 0.5 1 RCIM_Cubic>]}, YCurve {keys [<at 0 0 RCIM_Cubic>,<at 0.5 1 RCIM_Cubic>]}}}
     ScaleCurve=&Vector2DFromFloat(Value=1)
    ))
   &ScaleSpriteSizeBySpeed(Max Scale Factor=<1 50>, Min Scale Factor=<0.01 0.05>, Velocity Threshold=1)
   &ScaleColor(
    Scale Alpha=&FloatFromCurve(
     FloatCurve=DataInterface() {propertyValues {Curve {keys [
         <at 0 0 RCIM_Cubic>
         <at 0.3759 1.0207>
         {arriveTangent=-2.221, interpMode=RCIM_Cubic, leaveTangent=-2.221, tangentMode=RCTM_User, time=1, value=0}
        ]}}}
     Scale Curve=&MakeCustomFloatFromBool(
      Bool=$User.bIsFrontEndPreview
      True Float=&MakeCustomFloatFromBool(Bool=$User.bIsFullyDeployed, True Float=1, False Float=0)
      False Float=&Lerp_Float(A=0, B=1, Alpha=$System.FN_Integrator001.Value)
     )
    )
    Scale RGB=&VectorFromFloat(Value=1)
    ScaleA=true
    ScaleRGB=true
   )
  ]
  rendererClasses [&NiagaraSpriteRendererProperties]
  renderers [{rendererClass=&NiagaraSpriteRendererProperties}]
 }
 Wheels {
  emitterSpawnScript [/SetVariables(Emitter.AudioPlayer=DataInterface() {propertyValues {
      Attenuation=null
      Concurrency=null
      MaxPlaysPerTick=10
      SoundToPlay=/BRCosmetics/Sounds/Gliders/BariumFemale/GliderThrustStart_BariumFemale_Sideways_Cue
      bAllowLoopingOneShotSounds=false
      bLimitPlaysPerTick=true
      bOnlyActiveDuringGameplay=false
      bStopWhenComponentIsDestroyed=true
      ConfigurationUserParameter {parameter {name=None, typeDefHandle {registeredTypeIndex=1}, varData []}}
      ParameterNames []
     }}) {isSetParameters=true}]
  emitterUpdateScript [
   &EmitterState()
   &FN_Timer()
   &SpawnRate(SpawnRate=18, Use Spawn Probability=false)
   /Niagara/Modules/Audio/PlayPersistentAudio(
    Audio Player:NiagaraDataInterfaceAudioPlayer=$Emitter.AudioPlayer
    Enable Fade In / Out=true
    Fade Out Time=1.5
    Play Audio:NiagaraBool=&SetBoolByFloatComparison(A=&Abs_Float(Float:NiagaraFloat=$User.RightDot), B=0.5)
   )
   /Niagara/Modules/Audio/UpdatePersistentAudio(Audio Player:NiagaraDataInterfaceAudioPlayer=$Emitter.AudioPlayer, Update Type:ENiagara_AudioParamType=::NewEnumerator5)
  ]
  particleSpawnScript [
   &FN_SetSpawnGroup()
   &InitializeParticle(Color=$User.BariumPaintColor, Mesh Scale Mode=::NewEnumerator3, Mesh Uniform Scale=0.3, UsePositionOffset=false, Write Material Random=true)
   &InitialMeshOrientation(Mesh Orientation Mode=::NewEnumerator1, Use Rotation Vector=false) {enabled=false}
  ]
  particleUpdateScript [
   &ParticleState()
   &SkeletalMeshLocation(
    Bone / Socket Index=$Particles.SpawnGroup
    MaskSpawnGroup=false
    Mesh Sampling Type=::NewEnumerator5
    Sampled Position Offset=&RandomRangeVector(Minimum=<-4 0 0>, Maximum=<4 0 0>)
    Skeletal Mesh=$User.SkeletalMesh
    Socket Sampling Mode=::NewEnumerator5
   )
   &DynamicMaterialParameters(
    Index 2 Param 4=1
    Index 3 Param 1=&Add_Float(A=&RandomRangeFloat(Minimum=0, Maximum=1), B=$Emitter.FN_Timer.Value)
    Index 3 Param 2=0.07
    Module.Param0WriteEnabled2=true
    Module.Param0WriteEnabled3=true
    Module.Param1WriteEnabled2=true
    Module.Param1WriteEnabled3=true
    Module.Param2WriteEnabled2=true
    Module.Param2WriteEnabled3=true
    Module.Param3WriteEnabled2=true
    Module.Param3WriteEnabled3=true
    Write Parameter Index 0=false
    Write Parameter Index 2=true
    Write Parameter Index 3=true
   )
   &DynamicMaterialParameters(
    Index 2 Param 4=1
    Index 3 Param 2=0.05
    Module.Param0WriteEnabled2=true
    Module.Param0WriteEnabled3=true
    Module.Param1WriteEnabled2=true
    Module.Param1WriteEnabled3=true
    Module.Param2WriteEnabled2=true
    Module.Param2WriteEnabled3=true
    Module.Param3WriteEnabled2=true
    Module.Param3WriteEnabled3=true
    Write Parameter Index 0=false
    Write Parameter Index 2=true
    Write Parameter Index 3=true
   ) {enabled=false, module=DynamicMaterialParameters001}
   &ScaleColor(
    Scale Alpha=&FloatFromCurve(
     FloatCurve=DataInterface() {propertyValues {Curve {keys [
         {interpMode=RCIM_Cubic, tangentMode=RCTM_User, time=0, value=0}
         {interpMode=RCIM_Cubic, tangentMode=RCTM_User, time=0.5, value=1}
         {tangentMode=RCTM_User, time=1, value=0}
        ]}}}
     Scale Curve=&Lerp_Float(A=0.5, B=1, Alpha=$System.FN_Integrator001.Value)
    )
    ScaleA=true
    ScaleRGB=true
   )
   &UpdateMeshOrientation(Rotation Rate=0) {enabled=false}
   /SetVariables(Particles.MeshOrientation=&MultiplyQuaternion(
     Quaternion A=&MultiplyQuaternion(
      Quaternion A=$Particles.SkeletalMeshLocation.SampledRotation
      Quaternion B=&MakeQuatFromAxisAngle(Axis=<1 0 0>, AngleInDegrees=&RandomRangeFloat(Minimum=-360, Maximum=360))
     )
     Quaternion B=&MakeQuatFromAxisAngle(Axis=<1 0 0>, AngleInDegrees=&Multiply_Float(A=360, B=$Engine.Time), NormalizeAxis=true)
    )) {isSetParameters=true}
  ]
  rendererClasses [&NiagaraSpriteRendererProperties, &NiagaraMeshRendererProperties]
  renderers [
   {rendererClass=&NiagaraSpriteRendererProperties}
   {rendererClass=&NiagaraMeshRendererProperties, rendererIndex=1}
  ]
 }
 Wheels001 {
  emitterUpdateScript [
   &EmitterState(Life Cycle Mode=::NewEnumerator1, Loop Behavior=::NewEnumerator1, Loop Delay=3, Loop Duration Mode=::NewEnumerator1, UseLoopDelay=true)
   &FN_Timer()
   &SpawnBurst_Instantaneous(Spawn Count=4, Use Spawn Probability=false, UseLoopCountLimit=false)
  ]
  particleSpawnScript [
   &FN_SetSpawnGroup()
   &InitializeParticle(Color=$User.BariumPaintColor, Mesh Scale Mode=::NewEnumerator3, Mesh Uniform Scale=0.3, UsePositionOffset=false, Write Material Random=true)
   &InitialMeshOrientation(
    Mesh Orientation Mode=::NewEnumerator2
    Rotation=&RandomRangeVector(Minimum=<0 0 0>, Maximum=<1 0 0>)
    Use Rotation Vector=true
   ) {enabled=false}
  ]
  particleUpdateScript [
   &ParticleState(Kill Particles When Lifetime Has Elapsed=false, Let Infinitely Lived Particles Die When Emitter Deactivates=true)
   &SkeletalMeshLocation(
    Bone / Socket Index=$Particles.SpawnGroup
    MaskSpawnGroup=false
    Mesh Sampling Type=::NewEnumerator5
    Skeletal Mesh=$User.SkeletalMesh
    Socket Sampling Mode=::NewEnumerator5
   )
   &UpdateMeshOrientation(Rotation Rate=&Multiply_Float(A=0.01, B=$Engine.Time), Rotation Vector=$Engine.Owner.SystemXAxis) {enabled=false}
   /SetVariables(Particles.MeshOrientation=&MultiplyQuaternion(
     Quaternion A=&MultiplyQuaternion(
      Quaternion A=$Particles.SkeletalMeshLocation.SampledRotation
      Quaternion B=&MakeQuatFromAxisAngle(Axis=<1 0 0>, AngleInDegrees=&RandomRangeFloat(Minimum=-360, Maximum=360))
     )
     Quaternion B=&MakeQuatFromAxisAngle(Axis=<1 0 0>, AngleInDegrees=&Multiply_Float(A=360, B=$Engine.Time), NormalizeAxis=true)
    )) {isSetParameters=true}
   /SetVariables(Particles.MeshOrientation=&MultiplyQuaternion(
     Quaternion A=$Particles.SkeletalMeshLocation.SampledRotation
     Quaternion B=&MakeQuatFromAxisAngle(Axis=<1 0 0>, AngleInDegrees=&Multiply_Float(A=360, B=$Engine.Time), NormalizeAxis=true)
    )) {enabled=false, isSetParameters=true}
   &DynamicMaterialParameters(
    Index 2 Param 4=1
    Index 3 Param 1=&Add_Float(A=&RandomRangeFloat(Minimum=0, Maximum=1), B=$Emitter.FN_Timer.Value)
    Index 3 Param 2=0.13
    Module.Param0WriteEnabled2=true
    Module.Param0WriteEnabled3=true
    Module.Param1WriteEnabled2=true
    Module.Param1WriteEnabled3=true
    Module.Param2WriteEnabled2=true
    Module.Param2WriteEnabled3=true
    Module.Param3WriteEnabled2=true
    Module.Param3WriteEnabled3=true
    Write Parameter Index 0=false
    Write Parameter Index 2=true
    Write Parameter Index 3=true
   )
   &ScaleColor(Scale Alpha=&Lerp_Float(A=0.5, B=1, Alpha=$System.FN_Integrator001.Value), ScaleA=true, ScaleRGB=false)
  ]
  rendererClasses [&NiagaraSpriteRendererProperties, &NiagaraMeshRendererProperties]
  renderers [
   {rendererClass=&NiagaraSpriteRendererProperties}
   {rendererClass=&NiagaraMeshRendererProperties, rendererIndex=1}
  ]
 }
}
summary {
 systemName=NS_Glider_Barium
 emitters [
  {emitterName=SpeedLines, rendererClasses [&NiagaraSpriteRendererProperties]}
  {bEnabled=false, emitterName=DEBUG, rendererClasses [&NiagaraSpriteRendererProperties]}
  {emitterName=FN_AttachedRibbon, rendererClasses [&NiagaraRibbonRendererProperties]}
  {emitterName=Wheels, rendererClasses [&NiagaraSpriteRendererProperties, &NiagaraMeshRendererProperties]}
  {emitterName=Wheels001, rendererClasses [&NiagaraSpriteRendererProperties, &NiagaraMeshRendererProperties]}
  {emitterName=FN_AttachedRibbon001, rendererClasses [&NiagaraRibbonRendererProperties]}
  {emitterName=FN_AttachedRibbon002, rendererClasses [&NiagaraRibbonRendererProperties]}
  {emitterName=FN_AttachedRibbon003, rendererClasses [&NiagaraRibbonRendererProperties]}
 ]
 userVariables [
  User.BariumPaintColor {type=LinearColor, description="", defaultValue {struct=LinearColor, value=<rgba 1 1 1 1>}}
  User.bIsFrontEnd {type=NiagaraBool}
  User.bIsFrontEndPreview {type=NiagaraBool}
  User.bIsFullyDeployed {type=NiagaraBool}
  User.ForwardDot {type=NiagaraFloat, defaultValue=0, description=""}
  User.RibbonIDOffset {type=NiagaraInt32}
  User.RightDot {type=NiagaraFloat, defaultValue=0, description=""}
  User.SkeletalMesh {
   type=NiagaraDataInterfaceSkeletalMesh
   defaultValue {
    struct=NiagaraExt_VariableValue_DataInterface
    value {dataInterface=&self:NiagaraDataInterfaceSkeletalMesh_0}
   }
  }
 ]
}
